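# Impact sound cue for the Aqua Jacket pickaxe cosmetic, classed under impact sounds.
# A mixer layers a random Aqua Jacket impact mix with a quieter random enemy impact borrowed from the Bad Egg Male pickaxe.
# Each layer passes through its own modulator for slight pitch variation.
# The cue primes on load and references shared pickaxe impact attenuation, a property sheet and player melee weapon concurrency.
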SoundCue /BRCosmetics/Sounds/PickAxes/Aqua_Jacket/PA_AquaJacket_Impact_Cue
root=Mixer_0
nodes [
 {id=Mixer_0, children [Modulator_0,Modulator_1], properties {inputVolume [1,0.7]}}
 {id=Modulator_0, children [Random_0], properties {pitchMax=1, pitchMin=0.97}}
 {id=Modulator_1, children [Random_1], properties {pitchMax=0.99}}
 {id=Random_0, children [WavePlayer_0,WavePlayer_1,WavePlayer_6,WavePlayer_7], properties {weights [1,1,1,1]}}
 {id=Random_1, children [WavePlayer_9, WavePlayer_10], properties {weights [1,1]}}
 {id=WavePlayer_0, properties {soundWaveAssetPtr=&/Pickaxe_AquaJacket_Impact_Mix_01}}
 {id=WavePlayer_1, properties {soundWaveAssetPtr=&/Pickaxe_AquaJacket_Impact_Mix_02}}
 {id=WavePlayer_10, properties {soundWaveAssetPtr=/BRCosmetics/Sounds/PickAxes/BadEggMale/PickAxe_BadEgg_Male_EnemyImpact_02}}
 {id=WavePlayer_6, properties {soundWaveAssetPtr=&/Pickaxe_AquaJacket_Impact_Mix_04}}
 {id=WavePlayer_7, properties {soundWaveAssetPtr=&/Pickaxe_AquaJacket_Impact_Mix_05}}
 {id=WavePlayer_9, properties {soundWaveAssetPtr=/BRCosmetics/Sounds/PickAxes/BadEggMale/PickAxe_BadEgg_Male_EnemyImpact_01}}
]
properties {
 attenuationSettings=/Game/Sounds/Attenuation/Fort_Pickaxe_Impact_Att
 audioPropertiesSheet=/BRCosmetics/Sounds/Settings/PropertySheets/Pickaxe/APS_Cosmetic_Pickaxe_Impact
 bPrimeOnLoad=true
 maxDistance=4300
 soundClassObject=&Fort_Impacts
 volumeMultiplier=0.675
 attenuationOverrides {
  bAttenuateWithLPF=true
  distanceAlgorithm=Logarithmic
  falloffDistance=800
  lPFRadiusMax=800
  lPFRadiusMin=400
  reverbWetLevelMax=0.5
  reverbWetLevelMin=0.115
  attenuationShapeExtents {x=300}
 }
 concurrencyOverrides {bLimitToOwner=true, maxCount=2, resolutionRule=StopOldest}
 concurrencySet [&Player_Melee_Weapon_Conc]
}
